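; Seven variant of the Hoagie Choppa vehicle Blueprint, overriding the material behavior for local passengers.
; The passenger material swap applies masked or regular Seven materials to slots 0, 1, 2 and 4 through FortVehicleInterface, then refreshes damage state.
; Delegates cache each created MID, and aiming down sights sets the slot 0 MID's PlayerIsADS scalar.
; The cosmetic wrap event is overridden with an empty body.
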(event EventLocalPassengerADS
  (params (AimDownSights :type Boolean))
  (bind _choppamid_slot0 (Variables|Default|GetChoppaMID_Slot0))
  ;[1 1 1] Set Player Is ADS bool
  (bind _output_get (Variables|Default|SetPlayerIsADS AimDownSights))
  (if _output_get
    (Rendering|Material|SetScalarParameterValue _choppamid_slot0 "PlayerIsADS" 1.0 :!class /Script/Engine.MaterialInstanceDynamic)
    (else
      (Rendering|Material|SetScalarParameterValue _choppamid_slot0 "PlayerIsADS" 0.0 :!class /Script/Engine.MaterialInstanceDynamic))))

(event EventK2_ApplyCosmeticWrap
  (params (LoadedWrap :type Athena%20Item%20Wrap%20Definition%20Object%20Reference)))

(event EventLocalPassengerMatSwap
  (bind _islocalplayerpassenger (Variables|Default|GetIsLocalPlayerPassenger))
  (Class|FortVehicleInterface|ApplyMaterialandUseWrapifPossible (select _islocalplayerpassenger "/Hoagie/Seven_Choppa/Materials/MI_Seven_Choppa_Parts_Masked.MI_Seven_Choppa_Parts_Masked" "/Hoagie/Seven_Choppa/Materials/MI_Seven_Choppa_Parts.MI_Seven_Choppa_Parts") 0 (delegate Custom|OnChoppaMIDCreated_SlotZero))
  (Class|FortVehicleInterface|ApplyMaterialandUseWrapifPossible (select _islocalplayerpassenger "/Hoagie/Seven_Choppa/Materials/MI_Seven_Choppa_Parts_Masked.MI_Seven_Choppa_Parts_Masked" "/Hoagie/Seven_Choppa/Materials/MI_Seven_Choppa_Parts.MI_Seven_Choppa_Parts") 1 (delegate Custom|OnChoppaMIDCreated_Slot1))
  (Class|FortVehicleInterface|ApplyMaterialandUseWrapifPossible (select _islocalplayerpassenger "/Hoagie/Seven_Choppa/Materials/MI_Seven_Choppa_Body_Masked.MI_Seven_Choppa_Body_Masked" "/Hoagie/Seven_Choppa/Materials/MI_Seven_Choppa_Body.MI_Seven_Choppa_Body") 2 (delegate Custom|OnChoppaMIDCreated_Slot2))
  (Class|FortVehicleInterface|ApplyMaterialandUseWrapifPossible (select _islocalplayerpassenger "/Hoagie/Seven_Choppa/Materials/MI_Seven_Choppa_Glass_Masked.MI_Seven_Choppa_Glass_Masked" "/Hoagie/Seven_Choppa/Materials/MI_Seven_Choppa_Glass.MI_Seven_Choppa_Glass") 4 (delegate Custom|OnChoppaMIDCreated_Slot4))
  (CallFunction|UpdateDamageState (Variables|Default|GetDamageValue)))

(event Custom|OnChoppaMIDCreated_SlotZero
  (params (MID :type Material%20Instance%20Dynamic%20Object%20Reference))
  (Variables|Default|SetChoppaMID_Slot0 MID))

(event Custom|OnChoppaMIDCreated_Slot1
  (params (MID :type Material%20Instance%20Dynamic%20Object%20Reference))
  (Variables|Default|SetChoppaMID_Slot1 MID))

(event Custom|OnChoppaMIDCreated_Slot2
  (params (MID :type Material%20Instance%20Dynamic%20Object%20Reference))
  (Variables|Default|SetChoppaMID_Slot2 MID))

(event Custom|OnChoppaMIDCreated_Slot4
  (params (MID :type Material%20Instance%20Dynamic%20Object%20Reference))
  (Variables|Default|SetChoppaMID_Slot4 MID))

(fn ConstructionScript
  (|CallParentFunction))
